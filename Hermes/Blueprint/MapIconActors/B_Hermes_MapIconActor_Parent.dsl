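; Parent Blueprint actor for Hermes map icons, exposing control of its Fort minimap component's icon, visibility and scale.
; On BeginPlay the authority sets minimap viewable distance to the larger of the cached full-map and minimap distances, as relevancy range.
; Icon changes run only on authority; visibility changes have no authority check.
; Both flush net dormancy to replicate, as does BeginPlay, while scale changes do not.
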(event EventBeginPlay
  ;[0.007568 0.06955 0.07812] BeginPlay: Replicate to the client.
  (|CallParentFunction)
  (Networking|SwitchHasAuthority
    (:Authority
      (MiniMapData|SetMiniMapViewableDistance (Variables|Default|GetFortMiniMapComponent) (Math|Float|Max(Float) (Variables|/Script/FortniteGame.FortMiniMapComponent_TagsAndDistance|GetCachedMapViewableDistance (Variables|Default|GetFortMiniMapComponent)) (Variables|/Script/FortniteGame.FortMiniMapComponent_TagsAndDistance|GetCachedMiniMapViewableDistance (Variables|Default|GetFortMiniMapComponent)))) ; Set the relevancy distance of this object to whichever is greater, the full map or mini map viewable distance.
      (Networking|FlushNetDormancy))))

(event Custom|SetMapIconVisible
  (params (IsVisible :type Boolean))
  ;[1 1 1] Set Map Icon Visible
  (MiniMapData|SetMiniMapIndicatorIsVisible (Variables|Default|GetFortMiniMapComponent) IsVisible)
  (Networking|FlushNetDormancy))

(event Custom|SetMapIcon
  (params (MiniMapIcon :type Object%20Reference))
  ;[0.02466 0.1126 0.1198] SetMapIcon: Update the map icon.
  (Networking|SwitchHasAuthority
    (:Authority
      (MiniMapData|SetMiniMapIcon (Variables|Default|GetFortMiniMapComponent) MiniMapIcon)
      (Networking|FlushNetDormancy))))

(fn SetIconScale
  (params (IconScale :type Vector%202D%20Structure :default "(X=1.500000,Y=1.500000)"))
  ;[1 1 1] Set Icon Scale
  (MiniMapData|SetMiniMapIconScale (Variables|Default|GetFortMiniMapComponent) IconScale))

(fn ConstructionScript)
